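# Physics asset for the anti-grav variant of the Mastermind backpack cosmetic, defining a short dangling bone chain on the pack.
# Four zero-length capsules act as spheres: a larger one on the pack bone and smaller, slightly offset ones on three dyn_drop bones.
# Constraints link the drop bones in series from the pack, with wide swing on one axis and tight limits on the other swing and twist, giving a mostly hinge-like sway.
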PhysicsAsset /BRCosmetics/Accessories/FORT_Backpacks/M_MED_Mastermind_Backpack/Meshes/M_MED_Mastermind_Pack_AntiGrav_Physics
bodies {
 dyn_drop_01 [{shapeType=Capsule, length=0, radius=1, center=<0.25 0 0>}]
 dyn_drop_02 [{shapeType=Capsule, length=0, radius=1, center=<0.25 0 0>}]
 dyn_drop_03 [{shapeType=Capsule, length=0, radius=1, center=<0.25 0 0>}]
 pack [{shapeType=Capsule, length=0, radius=2}]
}
constraints [
 {bone1Name=dyn_drop_01, bone2Name=pack, swing1LimitDegrees=45, swing2LimitDegrees=2.5, twistLimitDegrees=5}
 {bone1Name=dyn_drop_02, bone2Name=dyn_drop_01, swing1LimitDegrees=45, swing2LimitDegrees=2.5, twistLimitDegrees=5}
 {bone1Name=dyn_drop_03, bone2Name=dyn_drop_02, swing1LimitDegrees=45, swing2LimitDegrees=2.5, twistLimitDegrees=5}
]
