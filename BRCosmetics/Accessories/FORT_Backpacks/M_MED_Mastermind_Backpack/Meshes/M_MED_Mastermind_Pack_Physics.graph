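# Physics asset for the Mastermind backpack cosmetic, providing simple capsule bodies and joints for its dangling and moving parts.
# Short chains of "drop" bones at the back, left and right, plus a single upper one, hang from the central pack body through swing and twist constraints.
# Three gauge bodies each attach directly to the pack with twist and second swing locked, so each rotates on one swing axis like a hinge.
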PhysicsAsset /BRCosmetics/Accessories/FORT_Backpacks/M_MED_Mastermind_Backpack/Meshes/M_MED_Mastermind_Pack_Physics
bodies {
 dyn_drop_bk_a_01 [{shapeType=Capsule, length=1.5, radius=0.7, center=<0.8 0 0>, rotation=<rot 90 0 0>}]
 dyn_drop_bk_a_02 [{shapeType=Capsule, length=1.5, radius=0.7, center=<0.8 0 0>, rotation=<rot 90 0 0>}]
 dyn_drop_bk_a_03 [{shapeType=Capsule, length=1.5, radius=0.7, center=<0.8 0 0>, rotation=<rot 90 0 0>}]
 dyn_drop_bk_b_01 [{shapeType=Capsule, length=2, radius=0.8, center=<1.2 0 0>, rotation=<rot 90 0 0>}]
 dyn_drop_bk_b_02 [{shapeType=Capsule, length=2, radius=0.8, center=<1.2 0 0>, rotation=<rot 90 0 0>}]
 dyn_drop_bk_b_03 [{shapeType=Capsule, length=2, radius=0.8, center=<1.2 0 0>, rotation=<rot 90 0 0>}]
 dyn_drop_bk_c_01 [{shapeType=Capsule, length=0.8, radius=0.6, center=<0.8 0 0>, rotation=<rot 90 0 0>}]
 dyn_drop_bk_c_02 [{shapeType=Capsule, length=0.8, radius=0.6, center=<0.8 0 0>, rotation=<rot 90 0 0>}]
 dyn_drop_l_01 [{shapeType=Capsule, length=2, radius=0.5, center=<1.5 0 0>, rotation=<rot 90 0 0>}]
 dyn_drop_l_02 [{shapeType=Capsule, length=2, radius=0.5, center=<1.5 0 0>, rotation=<rot 90 0 0>}]
 dyn_drop_r_01 [{shapeType=Capsule, length=2, radius=0.5, center=<1.5 0 0>, rotation=<rot 90 0 0>}]
 dyn_drop_r_02 [{shapeType=Capsule, length=2, radius=0.5, center=<1.5 0 0>, rotation=<rot 90 0 0>}]
 dyn_drop_up [{shapeType=Capsule, length=1, radius=0.5, center=<0.5 0 0>, rotation=<rot 90 0 0>}]
 dyn_gauge_01 [{shapeType=Capsule, length=1, radius=1, center=<-1 0.1 0>, rotation=<rot 90 0 0>}]
 dyn_gauge_02 [{shapeType=Capsule, length=2, radius=1, center=<-2 0.1 0>, rotation=<rot 90 0 0>}]
 dyn_gauge_03 [{shapeType=Capsule, length=2, radius=1, center=<-1.7 0.2 0>, rotation=<rot 90 0 0>}]
 pack [{shapeType=Capsule, length=2, radius=2}]
}
constraints [
 {bone1Name=dyn_drop_up, bone2Name=pack, swing1LimitDegrees=45, swing2LimitDegrees=45, twistLimitDegrees=5}
 {bone1Name=dyn_drop_l_01, bone2Name=pack, swing1LimitDegrees=45, swing2LimitDegrees=45, twistLimitDegrees=5}
 {bone1Name=dyn_drop_l_02, bone2Name=dyn_drop_l_01, swing1LimitDegrees=45, swing2LimitDegrees=45, twistLimitDegrees=5}
 {bone1Name=dyn_drop_r_01, bone2Name=pack, swing1LimitDegrees=45, swing2LimitDegrees=45, twistLimitDegrees=5}
 {bone1Name=dyn_drop_r_02, bone2Name=dyn_drop_r_01, swing1LimitDegrees=45, swing2LimitDegrees=45, twistLimitDegrees=5}
 {bone1Name=dyn_drop_bk_a_01, bone2Name=pack, swing1LimitDegrees=45, swing2LimitDegrees=45, twistLimitDegrees=5}
 {bone1Name=dyn_drop_bk_a_02, bone2Name=dyn_drop_bk_a_01, swing1LimitDegrees=45, swing2LimitDegrees=45, twistLimitDegrees=5}
 {bone1Name=dyn_drop_bk_a_03, bone2Name=dyn_drop_bk_a_02, swing1LimitDegrees=45, swing2LimitDegrees=45, twistLimitDegrees=5}
 {bone1Name=dyn_drop_bk_b_01, bone2Name=pack, swing1LimitDegrees=45, swing2LimitDegrees=45, twistLimitDegrees=5}
 {bone1Name=dyn_drop_bk_b_02, bone2Name=dyn_drop_bk_b_01, swing1LimitDegrees=45, swing2LimitDegrees=45, twistLimitDegrees=5}
 {bone1Name=dyn_drop_bk_b_03, bone2Name=dyn_drop_bk_b_02, swing1LimitDegrees=45, swing2LimitDegrees=45, twistLimitDegrees=5}
 {bone1Name=dyn_drop_bk_c_01, bone2Name=pack, swing1LimitDegrees=45, swing2LimitDegrees=45, twistLimitDegrees=5}
 {bone1Name=dyn_drop_bk_c_02, bone2Name=dyn_drop_bk_c_01, swing1LimitDegrees=45, swing2LimitDegrees=45, twistLimitDegrees=5}
 {
  bone1Name=dyn_gauge_01
  bone2Name=pack
  swing1LimitDegrees=90
  swing2LimitDegrees=45
  swing2Motion=Locked
  twistLimitDegrees=45
  twistMotion=Locked
 }
 {
  bone1Name=dyn_gauge_02
  bone2Name=pack
  swing1LimitDegrees=90
  swing2LimitDegrees=45
  swing2Motion=Locked
  twistLimitDegrees=45
  twistMotion=Locked
 }
 {
  bone1Name=dyn_gauge_03
  bone2Name=pack
  swing1LimitDegrees=90
  swing2LimitDegrees=45
  swing2Motion=Locked
  twistLimitDegrees=45
  twistMotion=Locked
 }
]
